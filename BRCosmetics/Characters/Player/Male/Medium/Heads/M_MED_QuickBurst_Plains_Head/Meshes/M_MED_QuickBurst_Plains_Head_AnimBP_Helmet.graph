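# Head AnimBP for M_MED_QuickBurst_Plains that limits head rotation while the matching helmet is worn.
# On customization load, it compares the Face part's mesh name to the helmet's to set Helmet_on.
# Each update negates and scales the owner mesh's head-to-neck_02 rotation delta from ref pose, applied as an additive bone-space counter-rotation on head.
# Blends bypass this without the helmet, during emotes or in the front end; alpha requires a valid pawn owner.
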AnimBlueprint /BRCosmetics/Characters/Player/Male/Medium/Heads/M_MED_QuickBurst_Plains_Head/Meshes/M_MED_QuickBurst_Plains_Head_AnimBP_Helmet
target_skeleton=/Game/Characters/Player/Common/Fortnite_Base_Head/Fortnite_Base_3L_Head_Skeleton
graphs [
 AnimGraph {nodes [
   n1 = "Transform (Modify) Bone - Bone: head"(
    ComponentPose=Local To Component(LocalPose=Use cached pose 'Helmet_branch'())
    Alpha=Get enableHeadRotLimitAlpha() -> (enableHeadRotLimitAlpha:double)
    Translation="0,0,0"
    Rotation=Get head rot() -> (head rot:Rotator)
    Scale="1,1,1"
   ) {
    rotationMode=BMM_Additive
    rotationSpace=BCS_BoneSpace
    alphaBoolBlend {blendInTime=0.1, blendOption=Linear, blendOutTime=0.2}
    boneToModify {boneName=head}
    rotation=<rot 50 50 50>
   }
   n2 = Component To Local(ComponentPose=@n1)
   n3 = Use cached pose 'Helmet_branch'()
   n4 = Get Head limit() -> (Head limit:bool)
   Output Pose(Result=Blend Poses by bool(
     BlendPose_0=Blend Poses by bool(BlendPose_0=@n2, BlendPose_1=@n3, bActiveValue=@n4) {blendType=Linear}
     BlendPose_1=Use cached pose 'Helmet_branch'()
     bActiveValue=Get Helmet_on() -> (Helmet_on:bool)
    ) {blendType=Linear})
   Save cached pose 'Helmet_branch'(Pose=Input Pose()) {cacheName=Helmet_branch}
   Get bIsPlayingEmote()
  ]}
 EventGraph {nodes [
   Event BlueprintUpdateAnimation()
   n1 = GetOwnerSkeletalMeshComponent()
   n2 = BreakRotator(InRot=BreakTransform(InTransform=GetDeltaTransformFromRefPose(self=@n1, BoneName=head, BaseName=neck_02)).Rotation)
   n3 = Get Y percentage() -> (Y percentage:float)
   n4 = Get X percentage() -> (X percentage:float)
   n5 = Get Z percentage() -> (Z percentage:float)
   > Set head rot(head rot:Rotator=MakeRotator(
     Roll=float * float!(A=@n2, B=-1, C:double=@n4) -> (ReturnValue)
     Pitch=float * float(A=@n2.Pitch, B=-1, C=@n3)
     Yaw=float * float(A=@n2.Yaw, B=-1, C=@n5)
    )) -> (Output_Get:Rotator)
   > n6 = Is Valid(InputObject=TryGetPawnOwner())
   Set enableHeadRotLimitAlpha(execute=@n6.Is Not Valid, enableHeadRotLimitAlpha:double=0) -> (Output_Get:double)
   Set enableHeadRotLimitAlpha(execute=@n6.Is Valid, enableHeadRotLimitAlpha=1)
   n7 = Get bIsInFrontEnd()
   n8 = Get bIsPlayingEmote()
   > Set Head limit(Head limit:bool=AND Boolean(A=NOT Boolean(A=@n8), B=NOT Boolean(A=@n7))) -> (Output_Get:bool)
   n9 = Event OnCharacterCustomizationLoaded()
   n10 = GetSkinnedAsset!(self:Skinned Mesh Component Object Reference=Get Array Item with EFort Custom Part Type(Array=@n9.CustomMeshParts, PartType=Face)) -> (ReturnValue:Skinned Asset Object Reference)
   > Set Helmet_on(Helmet_on:bool="Equal Exactly (String)"(A=GetDisplayName(Object=@n10), B=Get isHelmet() -> (isHelmet:String))) -> (Output_Get:bool)
   TryGetPawnOwner()
  ]}
]
variables [
 enableHeadRotLimitAlpha {type=number}
 Head limit {type=boolean}
 head rot {type=Rotator}
 Helmet_on {type=boolean}
 isHelmet {type=string, default=M_MED_QuickBurst_PlainsHelmet_FaceAcc}
 X percentage {type=number, default=0.8}
 Y percentage {type=number, default=0.8}
 Z percentage {type=number, default=0.8}
]
